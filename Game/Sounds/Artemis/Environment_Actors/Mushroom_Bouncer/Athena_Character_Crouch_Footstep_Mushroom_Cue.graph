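# Sound cue for first-person crouched boot footsteps on mushroom surfaces in Athena, filed with the Mushroom Bouncer environment sounds.
# A quality-level switch sends its first two slots to a random pick of eleven step waves and its last two to a pick of four.
# Each pool is equally weighted and plays through its own modulator.
# Output uses first-person walk footstep attenuation, the Athena foley sound class and local footstep concurrency.
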SoundCue /Game/Sounds/Artemis/Environment_Actors/Mushroom_Bouncer/Athena_Character_Crouch_Footstep_Mushroom_Cue
root=QualityLevel_0
nodes [
 {id=Modulator_0, children [Random_2]}
 {id=Modulator_1, children [Random_1]}
 {id=QualityLevel_0, children [Modulator_0,Modulator_0,Modulator_1,Modulator_1]}
 {
  id=Random_1
  children [WavePlayer_23, WavePlayer_26, WavePlayer_27, WavePlayer_28]
  properties {bShouldExcludeFromBranchCulling=true, weights [1,1,1,1]}
 }
 {
  id=Random_2
  children [
   WavePlayer_6
   WavePlayer_7
   WavePlayer_8
   WavePlayer_9
   WavePlayer_10
   WavePlayer_11
   WavePlayer_18
   WavePlayer_19
   WavePlayer_20
   WavePlayer_21
   WavePlayer_22
  ]
  properties {weights [1,1,1,1,1,1,1,1,1,1,1]}
 }
 {
  id=WavePlayer_10
  properties {soundWaveAssetPtr=&/FS_Boot_Default_1P_Mushroom_Step_04}
 }
 {
  id=WavePlayer_11
  properties {soundWaveAssetPtr=&/FS_Boot_Default_1P_Mushroom_Step_05}
 }
 {
  id=WavePlayer_18
  properties {soundWaveAssetPtr=&/FS_Boot_Default_1P_Mushroom_Step_06}
 }
 {
  id=WavePlayer_19
  properties {soundWaveAssetPtr=&/FS_Boot_Default_1P_Mushroom_Step_07}
 }
 {
  id=WavePlayer_20
  properties {soundWaveAssetPtr=&/FS_Boot_Default_1P_Mushroom_Step_08}
 }
 {
  id=WavePlayer_21
  properties {soundWaveAssetPtr=&/FS_Boot_Default_1P_Mushroom_Step_09}
 }
 {
  id=WavePlayer_22
  properties {soundWaveAssetPtr=&/FS_Boot_Default_1P_Mushroom_Step_10}
 }
 {
  id=WavePlayer_23
  properties {soundWaveAssetPtr=&/FS_Boot_Default_1P_Mushroom_Step_01}
 }
 {
  id=WavePlayer_26
  properties {soundWaveAssetPtr=&/FS_Boot_Default_1P_Mushroom_Step_02}
 }
 {
  id=WavePlayer_27
  properties {soundWaveAssetPtr=&/FS_Boot_Default_1P_Mushroom_Step_03}
 }
 {
  id=WavePlayer_28
  properties {soundWaveAssetPtr=&FS_Boot_Default_1P_Mushroom_Step_04}
 }
 {id=WavePlayer_6, properties {soundWaveAssetPtr=&/FS_Boot_Default_1P_Mushroom_Step_11}}
 {id=WavePlayer_7, properties {soundWaveAssetPtr=&FS_Boot_Default_1P_Mushroom_Step_01}}
 {id=WavePlayer_8, properties {soundWaveAssetPtr=&FS_Boot_Default_1P_Mushroom_Step_02}}
 {id=WavePlayer_9, properties {soundWaveAssetPtr=&FS_Boot_Default_1P_Mushroom_Step_03}}
]
properties {
 attenuationSettings=/Game/Sounds/Attenuation/Fort_Player_Footsteps_1P_Walk_Athena_Att
 maxDistance=2097152
 soundClassObject=&Fort_Player_Foley_1P_Athena
 volumeMultiplier=1.0025
 concurrencyOverrides {bLimitToOwner=true, maxCount=1, resolutionRule=StopOldest}
 concurrencySet [&Athena_Footstep_Local_Conc]
}
