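# Sound cue for a third-person character hard landing on a Mushroom Bouncer environment actor, using the enemy player foley sound class.
# A Random node picks among four equally weighted mushroom boot-landing waves and feeds a root Modulator left at its default settings.
# It references the shared third-person landing attenuation and global footstep concurrency, alongside concurrency override values for owner-limited, stop-oldest voice capping.
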SoundCue /Game/Sounds/Artemis/Environment_Actors/Mushroom_Bouncer/Athena_Character_LandHard_Mushroom_3P_Cue
root=Modulator_0
nodes [
 {id=Modulator_0, children [Random_1]}
 {id=Random_1, children [WavePlayer_3,WavePlayer_4,WavePlayer_5,WavePlayer_7], properties {weights [1,1,1,1]}}
 {id=WavePlayer_3, properties {soundWaveAssetPtr=&/FS_Boot_Default_3P_Mushroom_Land_01}}
 {id=WavePlayer_4, properties {soundWaveAssetPtr=&/FS_Boot_Default_3P_Mushroom_Land_02}}
 {id=WavePlayer_5, properties {soundWaveAssetPtr=&/FS_Boot_Default_3P_Mushroom_Land_03}}
 {id=WavePlayer_7, properties {soundWaveAssetPtr=&/FS_Boot_Default_3P_Mushroom_Land_04}}
]
properties {
 attenuationSettings=/Game/Sounds/Attenuation/Fort_Player_Land_3P_Athena_Att
 maxDistance=3584
 soundClassObject=&Fort_Player_Foley_3P_Enemy_Athena
 volumeMultiplier=0.7017
 concurrencyOverrides {bLimitToOwner=true, maxCount=4, resolutionRule=StopOldest}
 concurrencySet [&Athena_Footstep_Global_Conc]
}
