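# Sound cue for a hard character landing on a mushroom, filed under the Mushroom Bouncer environment actor sounds.
# A Random node picks one of four equally weighted first-person boot landing waves and feeds a Modulator at the root for variation.
# The cue plays boosted in the first-person player foley sound class, using player landing attenuation and a footstep landing concurrency set.
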SoundCue /Game/Sounds/Artemis/Environment_Actors/Mushroom_Bouncer/Athena_Character_LandHard_Mushroom_Cue
root=Modulator_1
nodes [
 {id=Modulator_1, children [Random_1]}
 {id=Random_1, children [WavePlayer_3,WavePlayer_4,WavePlayer_5,WavePlayer_7], properties {weights [1,1,1,1]}}
 {id=WavePlayer_3, properties {soundWaveAssetPtr=&/FS_Boot_Default_1P_Mushroom_Land_01}}
 {id=WavePlayer_4, properties {soundWaveAssetPtr=&/FS_Boot_Default_1P_Mushroom_Land_02}}
 {id=WavePlayer_5, properties {soundWaveAssetPtr=&/FS_Boot_Default_1P_Mushroom_Land_03}}
 {id=WavePlayer_7, properties {soundWaveAssetPtr=&/FS_Boot_Default_1P_Mushroom_Land_04}}
]
properties {
 attenuationSettings=/Game/Sounds/Attenuation/Fort_Player_Land_Athena_Att
 maxDistance=2097152
 soundClassObject=&Fort_Player_Foley_1P_Athena
 volumeMultiplier=2.5061
 concurrencyOverrides {maxCount=4, resolutionRule=PreventNew}
 concurrencySet [/Game/Sounds/Concurrency_Rules/Athena_Footstep_Land_Conc]
}
