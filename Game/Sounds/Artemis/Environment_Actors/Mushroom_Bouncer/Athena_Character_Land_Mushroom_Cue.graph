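# Sound cue for a character landing on a mushroom, filed under the Mushroom Bouncer environment actor sounds.
# A random node picks one of four equally weighted first-person boot landing waves and feeds it through a modulator at the root.
# It uses the player landing attenuation, the first-person player foley sound class, a boosted volume and a shared footstep landing concurrency set.
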SoundCue /Game/Sounds/Artemis/Environment_Actors/Mushroom_Bouncer/Athena_Character_Land_Mushroom_Cue
root=Modulator_0
nodes [
 {id=Modulator_0, children [Random_0]}
 {id=Random_0, children [WavePlayer_0,WavePlayer_1,WavePlayer_2,WavePlayer_7], properties {weights [1,1,1,1]}}
 {id=WavePlayer_0, properties {soundWaveAssetPtr=&/FS_Boot_Default_1P_Mushroom_Land_01}}
 {id=WavePlayer_1, properties {soundWaveAssetPtr=&/FS_Boot_Default_1P_Mushroom_Land_02}}
 {id=WavePlayer_2, properties {soundWaveAssetPtr=&/FS_Boot_Default_1P_Mushroom_Land_03}}
 {id=WavePlayer_7, properties {soundWaveAssetPtr=&/FS_Boot_Default_1P_Mushroom_Land_04}}
]
properties {
 attenuationSettings=/Game/Sounds/Attenuation/Fort_Player_Land_Athena_Att
 maxDistance=2097152
 soundClassObject=&Fort_Player_Foley_1P_Athena
 volumeMultiplier=1.3033
 concurrencyOverrides {maxCount=4, resolutionRule=PreventNew}
 concurrencySet [/Game/Sounds/Concurrency_Rules/Athena_Footstep_Land_Conc]
}
