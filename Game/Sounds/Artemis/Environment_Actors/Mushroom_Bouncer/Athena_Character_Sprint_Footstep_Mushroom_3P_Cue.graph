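# Sound Cue for third-person sprint footsteps on mushroom surfaces in Athena, routed to the enemy foley class with footstep attenuation and local and global concurrency sets.
# A quality-level root sends its first two inputs to a modulated random pick of eleven boot-on-mushroom step waves.
# Its last two inputs share a smaller random pool of the first four waves, trimming variety by quality setting.
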SoundCue /Game/Sounds/Artemis/Environment_Actors/Mushroom_Bouncer/Athena_Character_Sprint_Footstep_Mushroom_3P_Cue
root=QualityLevel_1
nodes [
 {id=Modulator_0, children [Random_0]}
 {id=Modulator_1, children [Random_3]}
 {id=QualityLevel_1, children [Modulator_0,Modulator_0,Modulator_1,Modulator_1]}
 {
  id=Random_0
  children [
   WavePlayer_0
   WavePlayer_1
   WavePlayer_2
   WavePlayer_3
   WavePlayer_4
   WavePlayer_5
   WavePlayer_6
   WavePlayer_7
   WavePlayer_8
   WavePlayer_9
   WavePlayer_10
  ]
  properties {weights [1,1,1,1,1,1,1,1,1,1,1]}
 }
 {id=Random_3, children [WavePlayer_11, WavePlayer_12, WavePlayer_13, WavePlayer_14], properties {weights [1,1,1,1]}}
 {id=WavePlayer_0, properties {soundWaveAssetPtr=&/FS_Boot_Default_3P_Mushroom_Step_01}}
 {id=WavePlayer_1, properties {soundWaveAssetPtr=&/FS_Boot_Default_3P_Mushroom_Step_02}}
 {
  id=WavePlayer_10
  properties {soundWaveAssetPtr=&/FS_Boot_Default_3P_Mushroom_Step_11}
 }
 {
  id=WavePlayer_11
  properties {soundWaveAssetPtr=&FS_Boot_Default_3P_Mushroom_Step_01}
 }
 {
  id=WavePlayer_12
  properties {soundWaveAssetPtr=&FS_Boot_Default_3P_Mushroom_Step_02}
 }
 {
  id=WavePlayer_13
  properties {soundWaveAssetPtr=&/FS_Boot_Default_3P_Mushroom_Step_03}
 }
 {
  id=WavePlayer_14
  properties {soundWaveAssetPtr=&/FS_Boot_Default_3P_Mushroom_Step_04}
 }
 {id=WavePlayer_2, properties {soundWaveAssetPtr=&FS_Boot_Default_3P_Mushroom_Step_03}}
 {id=WavePlayer_3, properties {soundWaveAssetPtr=&FS_Boot_Default_3P_Mushroom_Step_04}}
 {id=WavePlayer_4, properties {soundWaveAssetPtr=&/FS_Boot_Default_3P_Mushroom_Step_05}}
 {id=WavePlayer_5, properties {soundWaveAssetPtr=&/FS_Boot_Default_3P_Mushroom_Step_06}}
 {id=WavePlayer_6, properties {soundWaveAssetPtr=&/FS_Boot_Default_3P_Mushroom_Step_07}}
 {id=WavePlayer_7, properties {soundWaveAssetPtr=&/FS_Boot_Default_3P_Mushroom_Step_08}}
 {id=WavePlayer_8, properties {soundWaveAssetPtr=&/FS_Boot_Default_3P_Mushroom_Step_09}}
 {id=WavePlayer_9, properties {soundWaveAssetPtr=&/FS_Boot_Default_3P_Mushroom_Step_10}}
]
properties {
 attenuationSettings=/Game/Sounds/Attenuation/Fort_Player_Footsteps_3P_Walk_Athena_Att
 maxDistance=3584
 soundClassObject=&Fort_Player_Foley_3P_Enemy_Athena
 volumeMultiplier=0.7962
 concurrencyOverrides {bLimitToOwner=true, maxCount=2, resolutionRule=StopOldest}
 concurrencySet [&Athena_Footstep_Local_Conc, &Athena_Footstep_Global_Conc]
}
