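# Sound cue for the looping ambient sound of the mushroom launcher bouncer environment actor.
# It picks one of two looping waves at random, with equal weight, and randomizes its pitch and volume.
# A continuous modulator then drives pitch and volume from an "Emissive" parameter.
# The quality switch feeds this chain to its first two levels and leaves the others empty, with overridden attenuation on an outdoor ambient sound class.
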SoundCue /Game/Sounds/Artemis/Environment_Actors/Mushroom_Bouncer/Env_Mushroom_Launcher_AMB_Loop_Cue
root=QualityLevel_0
nodes [
 {
  id=ModulatorContinuous_0
  children [Modulator_0]
  properties {
   pitchModulationParams {maxOutput=1.1, minOutput=0.8, parameterName=Emissive}
   volumeModulationParams {maxOutput=1.2, minOutput=1, parameterName=Emissive}
  }
 }
 {id=Modulator_0, children [Random_0], properties {pitchMax=1, pitchMin=0.7, volumeMax=1, volumeMin=0.85}}
 {id=QualityLevel_0, children [ModulatorContinuous_0, ModulatorContinuous_0, null, null]}
 {id=Random_0, children [WavePlayer_1,WavePlayer_0], properties {weights [1,1]}}
 {
  id=WavePlayer_0
  properties {bLooping=true, soundWaveAssetPtr=&/Env_Mushroom_Launcher_AMB_Loop_01}
 }
 {
  id=WavePlayer_1
  properties {bLooping=true, soundWaveAssetPtr=&/Env_Mushroom_Launcher_AMB_Loop_02}
 }
]
properties {
 bOverrideAttenuation=true
 maxDistance=1200
 soundClassObject=&Fort_Ambient_Outdoors_Athena
 volumeMultiplier=0.752
 attenuationOverrides {falloffDistance=1000, attenuationShapeExtents {x=200}}
}
